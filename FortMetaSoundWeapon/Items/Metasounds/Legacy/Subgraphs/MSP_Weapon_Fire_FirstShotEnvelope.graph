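# MetaSound patch for legacy weapon fire sounds that outputs a FirstShotEnvelope multiplier.
# On each OnPlay, a trigger compare on bAutoFire either fires an AD envelope or selects a constant 1.
# The envelope has zero attack and a decay scaled from FireRate, and is remapped between EnvelopeMinMult and EnvelopeMaxMult.
# A float trigger route passes whichever branch fired to the output.
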MetaSoundPatch /FortMetaSoundWeapon/Items/Metasounds/Legacy/Subgraphs/MSP_Weapon_Fire_FirstShotEnvelope
inputs [
 bAutoFire {type=Bool}
 EnvelopeMaxMult {type=Float}
 EnvelopeMinMult {type=Float}
 FireRate {type=Float}
 OnPlay {type=Trigger}
]
outputs [FirstShotEnvelope {type=Float}]
pages [Default {
  input_defaults {EnvelopeMaxMult=1, EnvelopeMinMult=0.9, FireRate=0.2, OnPlay=null, bAutoFire=null}
  nodes [
   n1 = TriggerCompare.Clamp:Bool(Compare=^OnPlay, A=^bAutoFire) {also {defaults {B="<MetaSoundBuilderResult.SUCCEEDED: 0>"}}}
   ADEnvelopeADEnvelope_0 = AD Envelope.AD Envelope:Float(
    Decay Time=UE.ConversionFloatToTime(In=UE.Multiply:Float(PrimaryOperand=^FireRate) {also {defaults {AdditionalOperands=1.2}}}.out).out
    Trigger=@n1.TRUE
   ) {also {defaults {Attack Time=0, Hard Reset="<MetaSoundBuilderResult.SUCCEEDED: 0>"}}}
   TriggerRouteTriggerRouteFloat2_0 = "TriggerRoute.Trigger Route (Float, 2):Float"(
    Set 0=@ADEnvelopeADEnvelope_0.On Trigger
    Value 0=MapRange.MapRange:Float(Out Range B=^EnvelopeMaxMult, Out Range A=^EnvelopeMinMult, In=@ADEnvelopeADEnvelope_0.Out Envelope).Out Value
    Set 1=UE.Reroute(Value=@n1.FALSE).Value
   ) {also {defaults {Value 1=1}}}
  ]
  links [{dst="@FirstShotEnvelope", src=TriggerRouteTriggerRouteFloat2_0.Value}]
 }]
