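# MetaSound patch for legacy weapon fire sounds that signals when playback is almost finished and when it has finished.
# Stop starts a delay lasting the playing wave's duration, while accumulate uses a fixed shorter time.
# When that delay expires it updates the almost-finished bool and starts a second delay that fires OnFinished.
# Retrigger resets both delays, and both play and retrigger set the bool back through its trigger route.
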MetaSoundPatch /FortMetaSoundWeapon/Items/Metasounds/Legacy/Subgraphs/MSP_Weapon_Fire_OnFinishControl
inputs [
 OnAccumulate {type=Trigger}
 OnPlay {type=Trigger}
 OnRetrigger {type=Trigger}
 OnTriggerStop {type=Trigger}
 PlayingWaveDuration {type=Time}
]
outputs [
 bOnAlmostFinished {type=Bool}
 OnFinished {type=Trigger}
]
pages [Default {
  input_defaults {OnAccumulate=null, OnPlay=null, OnRetrigger=null, OnTriggerStop=null, PlayingWaveDuration=4}
  nodes [
   TriggerRouteTriggerRouteTime2_0 = "TriggerRoute.Trigger Route (Time, 2):Time"(Set 1=^OnAccumulate, Set 0=^OnTriggerStop, Value 0=^PlayingWaveDuration) {also {defaults {Value 1=1.5}}}
   TriggerDelay_1 = UE.Trigger Delay(Reset=^OnRetrigger, In=@TriggerRouteTriggerRouteTime2_0.On Set, Delay Time=@TriggerRouteTriggerRouteTime2_0.Value)
   TriggerDelay_0 = UE.Trigger Delay(Reset=^OnRetrigger, In=@TriggerDelay_1.out)
   TriggerRouteTriggerRouteBool3_0 = "TriggerRoute.Trigger Route (Bool, 3):Bool"(Set 1=^OnPlay, Set 2=^OnRetrigger, Set 0=@TriggerDelay_1.out) {also {defaults {Value 0="<MetaSoundBuilderResult.SUCCEEDED: 0>"}}}
  ]
  links [{dst="@OnFinished", src=TriggerDelay_0.out}, {dst="@bOnAlmostFinished", src=TriggerRouteTriggerRouteBool3_0.Value}]
 }]
