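# MetaSound patch that synthesizes a mono sub-frequency layer for weapon fire sounds.
# OnPlay passes through a bEnable comparison, then waits a random delay between RandomTimeMin and RandomTimeMax before triggering an audio-rate ADSR envelope.
# That envelope multiplies a wave-shaped sine at SubFrequency to form Out Mono, with attack, decay, sustain and shaping amount exposed as inputs.
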MetaSoundPatch /FortMetaSoundWeapon/Items/Metasounds/Legacy/Subgraphs/MSP_Weapon_Fire_Subsynth
inputs [
 Attack Curve {type=Float}
 Attack Time {type=Time}
 bEnable {type=Bool}
 Decay Curve {type=Float}
 Decay Time {type=Time}
 OnPlay {type=Trigger}
 RandomTimeMax {type=Time}
 RandomTimeMin {type=Time}
 SubFrequency {type=Float}
 Sustain Level {type=Float}
 WaveShapeAmount {type=Float}
]
interfaces [MetaSound:1.0]
outputs [Out Mono {type=Audio}]
pages [Default {
  input_defaults {
   Attack Curve=0.5
   Attack Time=0.01
   Decay Curve=0.5
   Decay Time=0.1
   OnPlay=null
   RandomTimeMax=0.05
   RandomTimeMin=0.05
   SubFrequency=35
   Sustain Level=0
   WaveShapeAmount=1
   bEnable="<MetaSoundBuilderResult.SUCCEEDED: 0>"
  }
  nodes [
   n1 = UE.RandomTime(
    Max=^RandomTimeMax
    Min=^RandomTimeMin
    Next=TriggerCompare.Clamp:Bool(Compare=^OnPlay, A=^bEnable) {also {defaults {B="<MetaSoundBuilderResult.SUCCEEDED: 0>"}}}.True
   )
   TriggerDelay_0 = UE.Trigger Delay(In=@n1.On Next, Delay Time=@n1.Value)
   ADSREnvelopeADSREnvelope_0 = ADSR Envelope.ADSR Envelope:Audio(
    Attack Curve=^Attack Curve
    Attack Time=^Attack Time
    Decay Curve=^Decay Curve
    Decay Time=^Decay Time
    Sustain Level=^Sustain Level
    Trigger Attack=@TriggerDelay_0.Out
   )
   Multiply_0 = UE.Multiply:Audio(
    PrimaryOperand=@ADSREnvelopeADSREnvelope_0.Out Envelope
    AdditionalOperands=UE.WaveShaper:Audio(Amount=^WaveShapeAmount, In=UE.Sine:Audio(Frequency=^SubFrequency).Audio) {also {defaults {Type=2}}}.Out
   )
  ]
  links [{dst="@Out Mono", src=Multiply_0.Out}]
 }]
