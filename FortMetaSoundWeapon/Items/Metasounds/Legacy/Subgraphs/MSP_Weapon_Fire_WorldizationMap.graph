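# MetaSound patch that turns the audio worldization enclosure modulation bus into an envelope value for weapon fire sounds.
# Each gunshot triggers an instant-attack AD envelope whose decay time is mapped from enclosure, moving from the max toward the min input as enclosure rises.
# A crossfade outputs a constant 1 at the activation threshold, blending toward the envelope as enclosure climbs above it.
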MetaSoundPatch /FortMetaSoundWeapon/Items/Metasounds/Legacy/Subgraphs/MSP_Weapon_Fire_WorldizationMap
inputs [
 ActivationThreshold {type=Float}
 EnvelopeMaxDecayTime {type=Float}
 EnvelopeMinDecayTime {type=Float}
 OnGunshot {type=Trigger}
]
outputs [WorlizationEnvelopeOut {type=Float}]
pages [Default {
  input_defaults {ActivationThreshold=0.9, EnvelopeMaxDecayTime=2, EnvelopeMinDecayTime=0.9, OnGunshot=null}
  nodes [
   n1 = Modulation.GetModulatorValue() {also {defaults {Modulator=&AudioWorldization_Bus_Enclosure}}}
   ADEnvelopeADEnvelope_0 = AD Envelope.AD Envelope:Float(
    Trigger=^OnGunshot
    Decay Time=UE.ConversionFloatToTime(In=MapRange.MapRange:Float(Out Range A=^EnvelopeMaxDecayTime, Out Range B=^EnvelopeMinDecayTime, In=@n1.Out).Out Value).out
   ) {also {defaults {Attack Time=0, Decay Curve=0.5, Hard Reset="<MetaSoundBuilderResult.SUCCEEDED: 0>"}}}
   CrossfadeTriggerRouteFloat2_0 = "Crossfade.Trigger Route (Float, 2):Float"(
    In 0=@ADEnvelopeADEnvelope_0.Out Envelope
    Crossfade Value=MapRange.MapRange:Float(In Range A=^ActivationThreshold, In=@n1.Out) {also {defaults {Out Range A=1, Out Range B=0}}}.Out Value
   ) {also {defaults {In 1=1}}}
  ]
  links [{dst="@WorlizationEnvelopeOut", src=CrossfadeTriggerRouteFloat2_0.out}]
 }]
