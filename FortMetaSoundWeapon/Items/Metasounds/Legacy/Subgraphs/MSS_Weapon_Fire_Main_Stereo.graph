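# Stereo one-shot MetaSound source in the legacy weapon subgraphs that plays the main fire sound from a supplied wave array with pitch and volume ranges.
# A trigger sequence alternates plays between two identical random-sound subgraphs, mixed to stereo so consecutive shots can overlap.
# Trigger controls pass only the most recently triggered player's finish event to OnFinished, so an older shot ending does not end the source.
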MetaSoundSource /FortMetaSoundWeapon/Items/Metasounds/Legacy/Subgraphs/MSS_Weapon_Fire_Main_Stereo
inputs [
 FireSounds {type=WaveAsset:Array}
 MaxPitch {type=Float}
 MaxVolume {type=Float}
 MinPitch {type=Float}
 MinVolume {type=Float}
 PitchMult {type=Float}
 UE.Source.OnPlay {type=Trigger}
]
interfaces [UE.OutputFormat.Stereo:1.0, UE.Source.OneShot:1.0, UE.Source:1.1]
outputs [
 "UE.OutputFormat.Stereo.Audio:0" {type=Audio}
 "UE.OutputFormat.Stereo.Audio:1" {type=Audio}
 UE.Source.OneShot.OnFinished {type=Trigger}
]
pages [Default {
  input_defaults {MaxPitch=1, MaxVolume=1, MinPitch=1, MinVolume=1, PitchMult=null, UE.Source.OnPlay=null, FireSounds []}
  nodes [
   n1 = UE.Reroute(Value=^FireSounds)
   TriggerSequence2_0 = "UE.Trigger Sequence (2)"(In=^UE.Source.OnPlay)
   n2 = &MSS_PlayRandomSound_Stereo(
    MaxPitch=^MaxPitch
    MaxVolume=^MaxVolume
    MinPitch=^MinPitch
    MinVolume=^MinVolume
    Pitch=^PitchMult
    Sounds=@n1.Value
    UE.Source.OnPlay=@TriggerSequence2_0.Out 0
   ) {also {defaults {Shared State Behavior=2}}}
   n3 = &MSS_PlayRandomSound_Stereo(
    MaxPitch=^MaxPitch
    MaxVolume=^MaxVolume
    MinPitch=^MinPitch
    MinVolume=^MinVolume
    Pitch=^PitchMult
    Sounds=@n1.Value
    UE.Source.OnPlay=@TriggerSequence2_0.Out 1
   ) {also {defaults {Shared State Behavior=2}}}
   AudioMixerAudioMixerStereo2_0 = "AudioMixer.Audio Mixer (Stereo, 2)"(
    In 0 L=@n2."UE.OutputFormat.Stereo.Audio:0"
    In 0 R=@n2."UE.OutputFormat.Stereo.Audio:1"
    In 1 L=@n3."UE.OutputFormat.Stereo.Audio:0"
    In 1 R=@n3."UE.OutputFormat.Stereo.Audio:1"
   )
   TriggerControl_0 = UE.Trigger Control(Trigger In=@n2.UE.Source.OneShot.OnFinished, Open=@TriggerSequence2_0.Out 0, Close=@TriggerSequence2_0.Out 1)
   TriggerControl_1 = UE.Trigger Control(Trigger In=@n3.UE.Source.OneShot.OnFinished, Close=@TriggerSequence2_0.Out 0, Open=@TriggerSequence2_0.Out 1)
   TriggerAnyTriggerAny2_0 = "TriggerAny.Trigger Any (2)"(In 0=@TriggerControl_0.Trigger Out, In 1=@TriggerControl_1.Trigger Out)
  ]
  links [
   {dst="@UE.OutputFormat.Stereo.Audio:0", src=AudioMixerAudioMixerStereo2_0.Out L}
   {dst="@UE.OutputFormat.Stereo.Audio:1", src=AudioMixerAudioMixerStereo2_0.Out R}
   {dst="@UE.Source.OneShot.OnFinished", src=TriggerAnyTriggerAny2_0.Out}
  ]
 }]
properties {
 assetImportData=&self:AssetImportData
 numChannels=2
 outputFormat=Stereo
 sampleRate=48000
}
