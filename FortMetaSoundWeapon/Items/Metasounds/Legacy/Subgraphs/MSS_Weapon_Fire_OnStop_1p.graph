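# Legacy 5.1 one-shot MetaSound subgraph for the first-person weapon fire stop sound.
# On play, it waits a random time, then MSS_PlayRandomSound_5_1 plays from OnStopSounds with random pitch and volume.
# OnStop or OnOutOfAmmo triggers an attack-decay envelope gating all six channels, so the already started sound becomes audible then.
# Decay lasts twice the wave duration plus the delay, and the player's finish ends the source.
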MetaSoundSource /FortMetaSoundWeapon/Items/Metasounds/Legacy/Subgraphs/MSS_Weapon_Fire_OnStop_1p
inputs [
 MaxPitch {type=Float}
 MaxVolume {type=Float}
 MinPitch {type=Float}
 MinVolume {type=Float}
 OnOutOfAmmo {type=Trigger}
 OnStop {type=Trigger}
 OnStopSounds {type=WaveAsset:Array}
 RandomTimeMax {type=Time}
 RandomTimeMin {type=Time}
 UE.Source.OnPlay {type=Trigger}
]
interfaces [UE.OutputFormat.5dot1:1.0, UE.Source.OneShot:1.0, UE.Source:1.1]
outputs [
 "UE.OutputFormat.5dot1.Audio:0" {type=Audio}
 "UE.OutputFormat.5dot1.Audio:1" {type=Audio}
 "UE.OutputFormat.5dot1.Audio:2" {type=Audio}
 "UE.OutputFormat.5dot1.Audio:3" {type=Audio}
 "UE.OutputFormat.5dot1.Audio:4" {type=Audio}
 "UE.OutputFormat.5dot1.Audio:5" {type=Audio}
 UE.Source.OneShot.OnFinished {type=Trigger}
]
pages [Default {
  input_defaults {
   MaxPitch=1.05
   MaxVolume=1.05
   MinPitch=0.95
   MinVolume=0.95
   OnOutOfAmmo=null
   OnStop=null
   RandomTimeMax=0
   RandomTimeMin=0
   UE.Source.OnPlay=null
   OnStopSounds []
  }
  nodes [
   InitVariable.Bool()
   InitVariable.Int32()
   InitVariable.Trigger()
   InitVariable.Trigger()
   n1 = UE.RandomTime(Max=^RandomTimeMax, Min=^RandomTimeMin, Next=^UE.Source.OnPlay)
   TriggerDelay_0 = UE.Trigger Delay(In=@n1.On Next, Delay Time=@n1.Value)
   MSS_PlayRandomSound_5_1_0 = /FortMetasound/Sources/MSS_PlayRandomSound_5_1(
    MaxPitch=^MaxPitch
    MaxVolume=^MaxVolume
    MinPitch=^MinPitch
    MinVolume=^MinVolume
    Sounds=UE.Reroute(Value=^OnStopSounds).Value
    UE.Source.OnPlay=@TriggerDelay_0.Out
   )
   TriggerAnyTriggerAny2_0 = "TriggerAny.Trigger Any (2)"(In 1=^OnOutOfAmmo, In 0=^OnStop)
   ADEnvelopeADEnvelope_0 = AD Envelope.AD Envelope:Float(
    Decay Time=UE.Add:Time(
     PrimaryOperand=UE.Multiply:Time by Float(PrimaryOperand=@MSS_PlayRandomSound_5_1_0.WaveDuration) {also {defaults {AdditionalOperands=2}}}.Out
     AdditionalOperands=UE.Reroute(Value=@n1.Value).Value
    ).Out
    Trigger=@TriggerAnyTriggerAny2_0.Out
   ) {also {defaults {Attack Curve=2, Attack Time=0.1}}}
   Multiply_0 = UE.Multiply:Audio by Float(AdditionalOperands=@ADEnvelopeADEnvelope_0.Out Envelope, PrimaryOperand=@MSS_PlayRandomSound_5_1_0."UE.OutputFormat.5dot1.Audio:0")
   Multiply_1 = UE.Multiply:Audio by Float(AdditionalOperands=@ADEnvelopeADEnvelope_0.Out Envelope, PrimaryOperand=@MSS_PlayRandomSound_5_1_0."UE.OutputFormat.5dot1.Audio:2")
   Multiply_2 = UE.Multiply:Audio by Float(AdditionalOperands=@ADEnvelopeADEnvelope_0.Out Envelope, PrimaryOperand=@MSS_PlayRandomSound_5_1_0."UE.OutputFormat.5dot1.Audio:3")
   Multiply_3 = UE.Multiply:Audio by Float(AdditionalOperands=@ADEnvelopeADEnvelope_0.Out Envelope, PrimaryOperand=@MSS_PlayRandomSound_5_1_0."UE.OutputFormat.5dot1.Audio:1")
   Multiply_4 = UE.Multiply:Audio by Float(AdditionalOperands=@ADEnvelopeADEnvelope_0.Out Envelope, PrimaryOperand=@MSS_PlayRandomSound_5_1_0."UE.OutputFormat.5dot1.Audio:4")
   Multiply_5 = UE.Multiply:Audio by Float(AdditionalOperands=@ADEnvelopeADEnvelope_0.Out Envelope, PrimaryOperand=@MSS_PlayRandomSound_5_1_0."UE.OutputFormat.5dot1.Audio:5")
  ]
  links [
   {dst="@UE.Source.OneShot.OnFinished", src=MSS_PlayRandomSound_5_1_0.UE.Source.OneShot.OnFinished}
   {dst="@UE.OutputFormat.5dot1.Audio:0", src=Multiply_0.Out}
   {dst="@UE.OutputFormat.5dot1.Audio:2", src=Multiply_1.Out}
   {dst="@UE.OutputFormat.5dot1.Audio:3", src=Multiply_2.Out}
   {dst="@UE.OutputFormat.5dot1.Audio:1", src=Multiply_3.Out}
   {dst="@UE.OutputFormat.5dot1.Audio:4", src=Multiply_4.Out}
   {dst="@UE.OutputFormat.5dot1.Audio:5", src=Multiply_5.Out}
  ]
 }]
properties {
 assetImportData=&self:AssetImportData
 numChannels=6
 outputFormat=FiveDotOne
 sampleRate=48000
}
