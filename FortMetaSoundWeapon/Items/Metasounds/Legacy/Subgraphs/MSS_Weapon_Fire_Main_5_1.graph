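# Legacy 5.1 MetaSound source subgraph that plays a weapon's main fire sound from a wave array with random pitch and volume.
# A trigger sequence alternates successive plays between two random-sound players, whose six channels are mixed so consecutive shots can overlap.
# Trigger controls pass OnFinished only from the most recently triggered player, and a trigger route reports that player's wave duration.
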MetaSoundSource /FortMetaSoundWeapon/Items/Metasounds/Legacy/Subgraphs/MSS_Weapon_Fire_Main_5_1
inputs [
 FireSounds {type=WaveAsset:Array}
 MaxPitch {type=Float}
 MaxVolume {type=Float}
 MinPitch {type=Float}
 MinVolume {type=Float}
 UE.Source.OnPlay {type=Trigger}
]
interfaces [UE.OutputFormat.5dot1:1.0, UE.Source.OneShot:1.0, UE.Source:1.1]
outputs [
 PlayingWaveDuration {type=Time}
 "UE.OutputFormat.5dot1.Audio:0" {type=Audio}
 "UE.OutputFormat.5dot1.Audio:1" {type=Audio}
 "UE.OutputFormat.5dot1.Audio:2" {type=Audio}
 "UE.OutputFormat.5dot1.Audio:3" {type=Audio}
 "UE.OutputFormat.5dot1.Audio:4" {type=Audio}
 "UE.OutputFormat.5dot1.Audio:5" {type=Audio}
 UE.Source.OneShot.OnFinished {type=Trigger}
]
pages [Default {
  input_defaults {MaxPitch=1.05, MaxVolume=1.05, MinPitch=0.95, MinVolume=0.95, UE.Source.OnPlay=null, FireSounds []}
  nodes [
   n1 = UE.Reroute(Value=^FireSounds)
   TriggerSequence2_0 = "UE.Trigger Sequence (2)"(In=^UE.Source.OnPlay)
   n2 = /FortMetasound/Sources/MSS_PlayRandomSound_5_1(
    MaxPitch=^MaxPitch
    MaxVolume=^MaxVolume
    MinPitch=^MinPitch
    MinVolume=^MinVolume
    Sounds=@n1.Value
    UE.Source.OnPlay=@TriggerSequence2_0.Out 1
   ) {also {defaults {Shared State Behavior=2}}}
   n3 = &MSS_PlayRandomSound_5_1(
    MaxPitch=^MaxPitch
    MaxVolume=^MaxVolume
    MinPitch=^MinPitch
    MinVolume=^MinVolume
    Sounds=@n1.Value
    UE.Source.OnPlay=@TriggerSequence2_0.Out 0
   ) {also {defaults {Shared State Behavior=2}}}
   AudioMixerAudioMixerMono2_0 = "AudioMixer.Audio Mixer (Mono, 2)"(In 1=@n2."UE.OutputFormat.5dot1.Audio:3", In 0=@n3."UE.OutputFormat.5dot1.Audio:3")
   AudioMixerAudioMixerMono2_1 = "AudioMixer.Audio Mixer (Mono, 2)"(In 1=@n2."UE.OutputFormat.5dot1.Audio:4", In 0=@n3."UE.OutputFormat.5dot1.Audio:4")
   AudioMixerAudioMixerMono2_2 = "AudioMixer.Audio Mixer (Mono, 2)"(In 1=@n2."UE.OutputFormat.5dot1.Audio:2", In 0=@n3."UE.OutputFormat.5dot1.Audio:2")
   AudioMixerAudioMixerMono2_3 = "AudioMixer.Audio Mixer (Mono, 2)"(In 1=@n2."UE.OutputFormat.5dot1.Audio:1", In 0=@n3."UE.OutputFormat.5dot1.Audio:1")
   AudioMixerAudioMixerMono2_4 = "AudioMixer.Audio Mixer (Mono, 2)"(In 1=@n2."UE.OutputFormat.5dot1.Audio:5", In 0=@n3."UE.OutputFormat.5dot1.Audio:5")
   AudioMixerAudioMixerMono2_5 = "AudioMixer.Audio Mixer (Mono, 2)"(In 1=@n2."UE.OutputFormat.5dot1.Audio:0", In 0=@n3."UE.OutputFormat.5dot1.Audio:0")
   TriggerControl_0 = UE.Trigger Control(Trigger In=@n2.UE.Source.OneShot.OnFinished, Close=@TriggerSequence2_0.Out 0, Open=@TriggerSequence2_0.Out 1)
   TriggerControl_1 = UE.Trigger Control(Trigger In=@n3.UE.Source.OneShot.OnFinished, Open=@TriggerSequence2_0.Out 0, Close=@TriggerSequence2_0.Out 1)
   TriggerAnyTriggerAny2_0 = "TriggerAny.Trigger Any (2)"(In 1=@TriggerControl_0.Trigger Out, In 0=@TriggerControl_1.Trigger Out)
   TriggerRouteTriggerRouteTime2_0 = "TriggerRoute.Trigger Route (Time, 2):Time"(Value 1=@n2.WaveDuration, Value 0=@n3.WaveDuration, Set 0=@TriggerSequence2_0.Out 0, Set 1=@TriggerSequence2_0.Out 1)
  ]
  links [
   {dst="@UE.OutputFormat.5dot1.Audio:3", src=AudioMixerAudioMixerMono2_0.Out}
   {dst="@UE.OutputFormat.5dot1.Audio:4", src=AudioMixerAudioMixerMono2_1.Out}
   {dst="@UE.OutputFormat.5dot1.Audio:2", src=AudioMixerAudioMixerMono2_2.Out}
   {dst="@UE.OutputFormat.5dot1.Audio:1", src=AudioMixerAudioMixerMono2_3.Out}
   {dst="@UE.OutputFormat.5dot1.Audio:5", src=AudioMixerAudioMixerMono2_4.Out}
   {dst="@UE.OutputFormat.5dot1.Audio:0", src=AudioMixerAudioMixerMono2_5.Out}
   {dst="@UE.Source.OneShot.OnFinished", src=TriggerAnyTriggerAny2_0.Out}
   {dst="@PlayingWaveDuration", src=TriggerRouteTriggerRouteTime2_0.Value}
  ]
 }]
properties {
 assetImportData=&self:AssetImportData
 numChannels=6
 outputFormat=FiveDotOne
 sampleRate=48000
}
